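; Clay tile vault wall static mesh actor that changes its material when a time-of-day change is requested.
; On BeginPlay it listens on the global GameplayEventRouter for OnTimeOfDayChangeRequested and sets a replicated SetMaterial flag.
; The flag's OnRep swaps material slot 0 to the A_B instance where cosmetic events can execute.
; The swap is one-way, as nothing here clears the flag or restores the original material.
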(event EventBeginPlay
  (|CallParentFunction)
  (GameplayEventRouter|StartListeningToEvent (GameplayEventRouter|GetGlobalEventRouter) "/Game/Conversation/Services/Effects/OnTimeOfDayChangeRequested.OnTimeOfDayChangeRequested"
    (:OnEventReceived
      (Variables|Replication|SetSetMaterial true))))

(event Collision|EventActorBeginOverlap
  (params (OtherActor :type Actor%20Object%20Reference)))

(event EventTick
  (params (DeltaSeconds :type Float%20%28single-precision%29)))

(fn OnRep_SetMaterial
  (Networking|CanExecuteCosmeticEvents
    (:True
      (Rendering|Material|SetMaterial (Variables|StaticMeshActor|GetStaticMeshComponent) 0 "/Hermes/Sets/ClayTile/Materials/MI_ClayTile_VaultWall_A_B.MI_ClayTile_VaultWall_A_B"))))

(fn ConstructionScript)
